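; Looping gameplay cue notify, filed under grenade launcher items, that adds a repeating Fortnite sound indicator while the cue is active.
; On application it calls the parent, then starts a looping timer at the configured interval and stores the handle.
; Each tick shows a local indicator tracking the target, reusing the application event's location and instigator, with audible distance taken from the configured sound.
; On removal it calls the parent and clears the timer.
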(event EventOnApplication(EitherVFX-preferthisone)
  (params (MyTarget :type Actor%20Object%20Reference) (Parameters :type Gameplay%20Cue%20Parameters%20Structure%20%28by%20ref%29) (ParticleComponents :type Array%20of%20FXSystem%20Component%20Object%20References) (AudioComponents :type Array%20of%20Audio%20Component%20Object%20References) (BurstCameraShakeInstance :type Camera%20Shake%20Base%20Object%20Reference) (BurstDecalInstance :type Decal%20Actor%20Object%20Reference))
  (|CallParentFunction MyTarget Parameters ParticleComponents AudioComponents BurstCameraShakeInstance BurstDecalInstance)
  (bind _returnvalue (Utilities|Time|SetTimerbyEvent (delegate Custom|ShowSoundIndicator) (Variables|SoundIndicator|GetInterval) true))
  (Variables|SoundIndicator|SetSoundIndicatorTimerHandle _returnvalue))

(event Custom|ShowSoundIndicator
  (Fortnite|ShowSoundIndicatorLocal :WorldContextObject (stale EventOnApplication(EitherVFX-preferthisone) MyTarget) :TrackedActor (stale EventOnApplication(EitherVFX-preferthisone) MyTarget) :Location (.Location (Ability|GameplayCue|BreakGameplayCueParameters (stale EventOnApplication(EitherVFX-preferthisone) Parameters))) :MaxAudibleDistance (Variables|/Script/Engine.SoundBase|GetMaxDistance (Variables|SoundIndicator|GetSound)) :IndicatorType (Variables|SoundIndicator|GetIndicator%20Type) :Instigator (.Instigator (Ability|GameplayCue|BreakGameplayCueParameters (stale EventOnApplication(EitherVFX-preferthisone) Parameters))) :Tint (Variables|SoundIndicator|GetTint) :OverrideIcon (Variables|SoundIndicator|GetOverride%20Icon) :SoundBase (Variables|SoundIndicator|GetSound) :AttenuationMaxRangeMultiplier 1.0 :AttenuationIntensityMultiplier 1.0))

(event EventOnRemoval(EitherVFX-preferthisversion)
  (params (MyTarget :type Actor%20Object%20Reference) (Parameters :type Gameplay%20Cue%20Parameters%20Structure%20%28by%20ref%29) (ParticleComponents :type Array%20of%20FXSystem%20Component%20Object%20References) (AudioComponents :type Array%20of%20Audio%20Component%20Object%20References) (BurstCameraShakeInstance :type Camera%20Shake%20Base%20Object%20Reference) (BurstDecalInstance :type Decal%20Actor%20Object%20Reference))
  (|CallParentFunction MyTarget Parameters ParticleComponents AudioComponents BurstCameraShakeInstance BurstDecalInstance)
  (Utilities|Time|ClearandInvalidateTimerbyHandle MyTarget (Variables|SoundIndicator|GetSoundIndicatorTimerHandle)))

(fn ConstructionScript)
