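; Looping gameplay cue for the Hopscotch grenade launcher's sticky projectile that shows a trail effect, named as the in-air state.
; On looping start it casts the target to the projectile, attaches the Niagara trail to its mesh and stores it as TrailFX.
; If the projectile's owner weapon is valid, its muzzle location feeds the trail's MuzzlePosition parameter.
; On removal it deactivates the stored trail, which is spawned with auto-destroy.
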(event EventOnLoopingStart(EitherVFX-preferthisversion)
  (params (MyTarget :type Actor%20Object%20Reference) (Parameters :type Gameplay%20Cue%20Parameters%20Structure%20%28by%20ref%29) (ParticleComponents :type Array%20of%20FXSystem%20Component%20Object%20References) (AudioComponents :type Array%20of%20Audio%20Component%20Object%20References))
  (|CallParentFunction MyTarget Parameters ParticleComponents AudioComponents)
  (bind _asb_prj_sticky_grenade_launcher_hopscotch_athena (Utilities|Casting|CastToB_Prj_Sticky_GrenadeLauncher_Hopscotch_Athena MyTarget))
  (bind _returnvalue (Effects|SpawnFXSystemAttached :FXSystemAsset "/HopscotchWeaponsGameplay/Items/GrenadeLauncher/FX/NS_GrenadeLauncher_HopScotch_Trail.NS_GrenadeLauncher_HopScotch_Trail" :AttachToComponent (Variables|/HopscotchWeaponsGameplay/Items/GrenadeLauncher/B_Prj_Sticky_GrenadeLauncher_Hopscotch_Athena.B_Prj_Sticky_GrenadeLauncher_Hopscotch_Athena_C|GetMesh _asb_prj_sticky_grenade_launcher_hopscotch_athena) :AttachPointName "None" :Rotation "90.000000,0.000000,0.000000" :Scale "1.000000,1.000000,1.000000" :LocationType "KeepRelativeOffset" :bAutoDestroy true :PoolingMethod "None" :bAutoActivate true))
  (bind _output_get (Variables|Default|SetTrailFX _returnvalue))
  (bind _returnvalue_2 (Class|FortProjectileBase|GetOwnerWeapon _asb_prj_sticky_grenade_launcher_hopscotch_athena))
  (Utilities|IsValid _returnvalue_2
    (:Is%20Valid
      (Effects|Components|ParticleSystem|SetVectorParameter _output_get "MuzzlePosition" (Effects|GetMuzzleLocation _returnvalue_2 -1)))))

(event EventOnRemoval(EitherVFX-preferthisversion)
  (params (MyTarget :type Actor%20Object%20Reference) (Parameters :type Gameplay%20Cue%20Parameters%20Structure%20%28by%20ref%29) (ParticleComponents :type Array%20of%20FXSystem%20Component%20Object%20References) (AudioComponents :type Array%20of%20Audio%20Component%20Object%20References) (BurstCameraShakeInstance :type Camera%20Shake%20Base%20Object%20Reference) (BurstDecalInstance :type Decal%20Actor%20Object%20Reference))
  (|CallParentFunction MyTarget Parameters ParticleComponents AudioComponents BurstCameraShakeInstance BurstDecalInstance)
  (bind _trailfx (Variables|Default|GetTrailFX)
    (:then
      (Components|Activation|Deactivate _trailfx))
    (:else)))

(fn ConstructionScript)
